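# Physics asset for the TacticKale backpack cosmetic, defining collision bodies and joint limits for its dynamic bones.
# A box body on the pack center acts as the hub, with capsules on two boom bones, two straps and a zippo bone each constrained directly to it.
# The booms have tight angular limits, the zippo swings more freely, and the straps have asymmetric left and right swing limits on a single axis.
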PhysicsAsset /BRCosmetics/Accessories/FORT_Backpacks/Backpack_M_MED_TacticKale/Meshes/M_MED_TacticKale_Pack_Physics
bodies {
 dyn_pack_boom_1 [{shapeType=Capsule, length=5.6254, radius=2.4072, shapeName=dyn_pack_boom_1_capsule, center=<0.126 0 -0.0736>, rotation=<rot 0 0 -90>}]
 dyn_pack_boom_2 [{shapeType=Capsule, length=5.6254, radius=2.4066, shapeName=dyn_pack_boom_2_capsule, center=<0.0026 0 0.1896>, rotation=<rot 0 0 -90>}]
 dyn_pack_center [{shapeType=Box, extentX=17.94, extentY=8.5407, extentZ=22.7407, shapeName=dyn_pack_center_box, center=<0 0 -0.0887>}]
 dyn_pack_strap_l [{shapeType=Capsule, length=1.7673, radius=1.3149, shapeName=dyn_pack_strap_l_capsule, center=<1.54 0 0>}]
 dyn_pack_strap_r [{shapeType=Capsule, length=3.1673, radius=1.3149, shapeName=dyn_pack_strap_r_capsule, center=<-1.6618 0 0>}]
 dyn_pack_zippo [{
   shapeType=Capsule
   length=1.505
   radius=0.605
   shapeName=dyn_pack_zippo_capsule
   center=<1.0443 -0.2701 0>
   rotation=<rot 90 160.5288 160.5287>
  }]
}
constraints [
 {bone1Name=dyn_pack_boom_1, bone2Name=dyn_pack_center, swing1LimitDegrees=10, swing2LimitDegrees=10, twistLimitDegrees=10}
 {bone1Name=dyn_pack_zippo, bone2Name=dyn_pack_center, swing1LimitDegrees=45, swing2LimitDegrees=45, twistLimitDegrees=10}
 {bone1Name=dyn_pack_strap_r, bone2Name=dyn_pack_center, swing1LimitDegrees=80, swing2LimitDegrees=0, twistLimitDegrees=0}
 {bone1Name=dyn_pack_strap_l, bone2Name=dyn_pack_center, swing1LimitDegrees=60, swing2LimitDegrees=0, twistLimitDegrees=0}
 {bone1Name=dyn_pack_boom_2, bone2Name=dyn_pack_center, swing1LimitDegrees=10, swing2LimitDegrees=10, twistLimitDegrees=10}
]
